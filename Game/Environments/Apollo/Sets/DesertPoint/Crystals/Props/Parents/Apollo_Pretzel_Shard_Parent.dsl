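; Parent Blueprint for pretzel shard props in the Apollo DesertPoint crystal set.
; BeginPlay only calls the parent class implementation, and the construction script is empty.
; The class adds no logic of its own in these graphs, so any behavior comes from its parent class.
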(event EventBeginPlay
  (|CallParentFunction))

(fn ConstructionScript)
